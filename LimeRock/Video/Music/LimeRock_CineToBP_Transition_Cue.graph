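# LimeRock sound cue that plays the Dent event cinematic-to-game transition audio, judging by its name and wave asset.
# A single wave player node is the root, with no mixing, randomization or other nodes.
# The cue is routed to the MediaCinematics sound class, with a raised volume multiplier and a very large max distance.
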SoundCue /LimeRock/Video/Music/LimeRock_CineToBP_Transition_Cue
root=WavePlayer_1
nodes [{id=WavePlayer_1, properties {soundWaveAssetPtr=&/241079_FNBR_Dent_Event_Cinematic_V204_30fps_1440p_Cine_to_Game_Audio}}]
properties {maxDistance=2097152, soundClassObject=/Game/Sounds/FortSoundClassesAndModes/MediaCinematics, volumeMultiplier=2}
